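# EQS query for GnatGala AI that picks a warp destination, intended to be ahead of its target.
# Generates navmesh-projected points on a raised 45-degree arc around the warp target, avoiding smashables and water.
# Tests trace and path to the target, favor points outside player vision cones, and filter by dot product against the current target's facing.
# A minimum 2D distance from the querier and a random test further filter candidates.
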EnvQuery /FNM_AI_GnatGala/BTInjectionBehaviour/EQS_Warp_GnatGala_MoveAheadofTarget
options [{
  generator {
   name=EnvQueryGenerator_OnCircle_0
   class=EnvQueryGenerator_OnCircle
   properties {
    arcAngle=45
    bDefineArc=true
    circleCenter=/NPCLibrary/Abilities/Warp/EQS/EnvQueryContext_WarpGetNPCTarget.EnvQueryContext_WarpGetNPCTarget_C
    circleCenterZOffset=360
    circleRadius=2250
    numberOfPoints=32
    pointOnCircleSpacingMethod=BySpaceBetween
    spaceBetween=2000
    projectionData {
     extentX=0
     navigationFilter=&NavFilter_Phoebe_NoSmashableNoWater_C
     postProjectionVerticalOffset=0
     projectDown=150
     projectUp=150
    }
    traceData {navigationFilter=null, traceMode=None}
   }
  }
  tests [
   EnvQueryTest_Trace_0 {
    class=EnvQueryTest_Trace
    properties {
     context=&EnvQueryContext_WarpGetNPCTarget_C
     scoringFactor=1
     testComment=""
     testPurpose=FilterAndScore
    }
   }
   FortQueryTest_IsInAnyPlayerVisionCone_0 {
    class=FortQueryTest_IsInAnyPlayerVisionCone
    properties {
     bDefineReferenceValue=false
     boolValue=false
     clampMaxType=None
     clampMinType=None
     fieldOfView=90
     filterType=Match
     floatValueMax=0
     floatValueMin=0
     maxDistance=5000
     multipleContextFilterOp=AllPass
     multipleContextScoreOp=AverageScore
     normalizationType=Absolute
     referenceValue=0
     scoreClampMax=0
     scoreClampMin=0
     scoringEquation=Constant
     scoringFactor=1
     testComment=""
     testPurpose=FilterAndScore
     useLineOfSight=false
    }
   }
   EnvQueryTest_Dot_0 {
    class=EnvQueryTest_Dot
    properties {
     clampMaxType=None
     clampMinType=None
     floatValueMax=0.5
     floatValueMin=1
     scoreClampMax=0
     scoreClampMin=0
     scoringFactor=1
     testComment=""
     testPurpose=Filter
     lineA {rotation=&FortAthenaAIBot_EQSQueryContext_CurrentTarget}
     lineB {lineFrom=&FortAthenaAIBot_EQSQueryContext_CurrentTarget}
    }
   }
   EnvQueryTest_Pathfinding_2 {
    class=EnvQueryTest_Pathfinding
    properties {
     context=&EnvQueryContext_WarpGetNPCTarget_C
     filterClass=&NavFilter_NPC_AnyTraversable_C
    }
   }
   EnvQueryTest_Distance_0 {
    class=EnvQueryTest_Distance
    properties {
     distanceTo=&EnvQueryContext_Querier
     filterType=Minimum
     floatValueMin=400
     scoringEquation=Linear
     scoringFactor=1
     testComment=""
     testMode=Distance2D
     testPurpose=Filter
    }
   }
   FortQueryTest_Random_0 {class=FortQueryTest_Random, properties {floatValueMax=1, scoringFactor=1, testPurpose=Filter}}
  ]
 }]
