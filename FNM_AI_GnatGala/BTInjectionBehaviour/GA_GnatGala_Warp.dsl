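; GA_GnatGala_Warp is a gameplay ability for the GnatGala AI that, judging by its name, performs a warp.
; Activation defers entirely to the parent class, so this asset only adds a gating rule.
; CanActivateAbility allows the ability only when the avatar actor's location lies outside every player's vision cone within a long range, so it cannot start while a player could see it.
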(event Ability|EventActivateAbility
  (|CallParentFunction))

(event Ability|EventOnEndAbility
  (params (bWasCancelled :type Boolean)))

(fn CanActivateAbility
  (params (ActorInfo :type Gameplay%20Ability%20Actor%20Info%20Structure) (Handle :type Gameplay%20Ability%20Spec%20Handle%20Structure))
  (bind (_owneractor _avataractor) (Utilities|Struct|BreakGameplayAbilityActorInfo ActorInfo))
  (bind _returnvalue (AI|IsPositioninAnyPlayerVisionCone (Transformation|GetActorLocation _avataractor) 1e+04))
  (return 0 (not _returnvalue)))
